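# Sound cue for interacting with the TV scare prop in the Troop Factor Wild Estate content.
# A mixer layers three waves at once with equal input volume: an interact sound, a TV turn-on and a static loop clip.
# It routes to the in-game SFX class with a slight volume boost and overrides concurrency to cap simultaneous instances, stopping the oldest.
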SoundCue /WildEstate_TroopFactor/Assets/Audio/SC_TVScare_TroopFactor_Interact
root=Mixer_0
nodes [
 {id=Mixer_0, children [WavePlayer_3,WavePlayer_5,WavePlayer_4], properties {inputVolume [1,1,1]}}
 {id=WavePlayer_3, properties {soundWaveAssetPtr=/WildEstate_TroopFactor/Assets/Audio/wav/FNM_TV_Interact_03}}
 {id=WavePlayer_4, properties {soundWaveAssetPtr=/WildEstate_TroopFactor/Assets/Audio/wav/FNM_TV_Static_loop}}
 {id=WavePlayer_5, properties {soundWaveAssetPtr=/WildEstate_TroopFactor/Assets/Audio/wav/FNM_TV_TurnOn}}
]
properties {
 bOverrideConcurrency=true
 maxDistance=2097152
 soundClassObject=&Fort_SFX_InGame
 volumeMultiplier=1.2
 concurrencyOverrides {maxCount=2, resolutionRule=StopOldest, retriggerTime=1}
}
