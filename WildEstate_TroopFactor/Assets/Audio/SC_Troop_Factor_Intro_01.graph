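# Sound cue for the Troop Factor intro sound, played at boosted volume in the creatures sound class and primed on load.
# A single wave player feeds a Modulator for slight variation, with the minimum pitch raised so it never drops below normal.
# Spatialization and voice limiting reference shared melee assets: Deimos attack attenuation and player melee weapon concurrency.
# Local attenuation and concurrency override values are stored, but their override switches are not enabled.
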SoundCue /WildEstate_TroopFactor/Assets/Audio/SC_Troop_Factor_Intro_01
root=Modulator_1
nodes [
 {id=Modulator_1, children [WavePlayer_15], properties {pitchMin=1}}
 {id=WavePlayer_15, properties {soundWaveAssetPtr=/WildEstate_TroopFactor/Assets/Audio/wav/Troop_Factor_Intro_01}}
]
properties {
 attenuationSettings=/Game/Sounds/Attenuation/Fort_Deimos_Attack_Melee_Att
 bPrimeOnLoad=true
 maxDistance=3300
 soundClassObject=/Game/Sounds/FortSoundClassesAndModes/Fort_Creatures
 volumeMultiplier=1.8
 attenuationOverrides {
  bAttenuateWithLPF=true
  distanceAlgorithm=Logarithmic
  falloffDistance=800
  lPFRadiusMax=800
  lPFRadiusMin=400
  attenuationShapeExtents {x=300}
 }
 concurrencyOverrides {bLimitToOwner=true, maxCount=2, resolutionRule=StopOldest}
 concurrencySet [&Player_Melee_Weapon_Conc]
}
