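# Translucent unlit Niagara material for the TroopFactor spike glow effect, built on a polar-mapped spike mask.
# A panning Voronoi texture, shaped by two power and intensity pairs, warps one polar axis while the other tiles and scrolls.
# The warped coordinates sample the mask using derivatives from plain UVs, which avoids mip seams at the polar wrap.
# The mask is multiplied by particle alpha, then depth-faded.
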Material /WildEstate_TroopFactor/Assets/VFX/M_FX_TroopFactor_SpikeGlow_01
graph {
 properties {
  bAutomaticallySetUsageInEditor=false
  bUsedWithNiagaraMeshParticles=true
  bUsedWithNiagaraRibbons=true
  bUsedWithNiagaraSprites=true
  blendMode=BLEND_Translucent
  shadingModel=MSM_Unlit
  usedShadingModels=Unlit
 }
 nodes [
  n1 = TextureCoordinate()
  n2 = MaterialFunctionCall!(UV Coord=@n1, LenghtScale=Constant() {r=2}, RadialScale=Constant() {r=0.5}) {materialFunction=/Game/Global/MaterialFunctions/TechArt/MF_PolarCoordinates}
  n3 = AppendVector(
   A=ScalarParameter() {defaultValue=0.1, parameterName=Tiling_Distortion-X}
   B=ScalarParameter() {defaultValue=1, parameterName=Tiling_Distortion-Y}
  )
  n4 = Multiply(A=AppendVector(A=Constant() {r=1}, B=Constant() {r=0.25}), B=@n2.Result)
  n5 = TextureSampleParameter2D!(
   UVs=Panner(Coordinate=Multiply(A=@n3, B=@n4)) {speedX=-0.1, speedY=0.02}
   "DDX(UVs)"=DDX(Value=@n1)
   "DDY(UVs)"=DDY(Value=@n1)
   Apply View MipBias
  ) {
   mipValueMode=TMVM_Derivative
   parameterName=Texture_Distortion
   samplerType=SAMPLERTYPE_LinearGrayscale
   texture=/Game/Packages/Fortress_Sky/SkyDome/Master/T_SkyDome_Cloud_Voroni
  }
  n6 = Multiply(
   A=Power(Base=@n5.R, Exp=ScalarParameter() {defaultValue=4, parameterName=Distortion-Texture_Power-A})
   B=ScalarParameter() {defaultValue=-0.4, parameterName=Distortion_Intensity-A}
  )
  n7 = Multiply(
   A=Power(Base=@n5.R, Exp=ScalarParameter() {defaultValue=10, parameterName=Distortion-Texture_Power})
   B=ScalarParameter() {defaultValue=-0.25, parameterName=Distortion_Intensity}
  )
  n8 = ComponentMask(None=Add(A=@n2.Result, B=Add(A=@n6, B=@n7))) {r=true}
  n9 = Add(A=ComponentMask(None=@n2.Result) {r=true}, B=Multiply(A=@n8, B=Constant() {r=0.5}))
  n10 = ComponentMask(None=@n2.Result) {g=true}
  n11 = ScalarParameter() {defaultValue=1, parameterName=Tiling}
  n12 = Panner(Coordinate=AppendVector(A=Clamp(None=@n9), B=Multiply(A=@n10, B=@n11))) {speedY=0.1}
  n13 = ComponentMask(None=@n12) {r=true}
  n14 = TextureCoordinate()
  n15 = TextureSampleParameter2D(
   UVs=AppendVector(A=Clamp(None=@n13) {maxDefault=0.98, minDefault=0.01}, B=ComponentMask(None=@n12) {g=true})
   "DDX(UVs)"=DDX(Value=@n14)
   "DDY(UVs)"=DDY(Value=@n14)
  ) {
   mipValueMode=TMVM_Derivative
   parameterName=Param
   texture=/BRCosmetics/Weapons/FORT_Melee/Pickaxe_TigerRootFame/FX/T_TigerRootFame_Spikes_Mask
  }
  DepthFade(Opacity=Multiply(A=@n15.R, B=ParticleColor().A), FadeDistance=ScalarParameter() {defaultValue=20, parameterName=DepthFade_Distance})
  n16 = Abs(None=@n8)
  OneMinus(None=Modulo!(A=Round(None=@n16), B=Constant() {r=2}))
 ]
}
