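# Additive, unlit, two-sided particle material for a TV glow effect in TroopFactor, usable on Niagara sprites and mesh particles.
# An exponential radial gradient is multiplied by an inverted center sphere mask, hollowing the glow, with parameters for density, radius and hardness.
# Particle color tints this shape, which feeds the shared MF_FX_EmissiveAdjustments function.
# Particle alpha times the shape gets a depth fade, which a quality switch skips on Low.
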Material /WildEstate_TroopFactor/Assets/VFX/M_FX_TroopFactor_TVGlow
graph {
 properties {
  bUsedWithNiagaraMeshParticles=true
  bUsedWithNiagaraSprites=true
  blendMode=BLEND_Additive
  pixelDepthOffsetMode=PDOM_Legacy
  shadingModel=MSM_Unlit
  twoSided=true
  usedShadingModels=Unlit
 }
 nodes [
  n1 = ParticleColor()
  n2 = TextureCoordinate()
  n3 = SphereMask(
   A=@n2
   B=Constant() {r=0.5}
   Radius=ScalarParameter() {defaultValue=0.35, parameterName=Radius}
   Hardness=ScalarParameter() {defaultValue=0.5, parameterName=Hardness}
  )
  n4 = Multiply(
   A=MaterialFunctionCall(UVs=@n2, Density=ScalarParameter() {defaultValue=1, parameterName=Density}) {materialFunction=&RadialGradientExponential}.RadialGradientExponential
   B=OneMinus(None=Multiply(A=@n3))
  )
  MaterialFunctionCall(Emissive=Multiply(A=@n1.RGB, B=@n4)) {materialFunction=&MF_FX_EmissiveAdjustments}
  n5 = Multiply(A=@n1.A, B=@n4)
  n6 = DepthFade(Opacity=@n5, FadeDistance=ScalarParameter() {defaultValue=20, parameterName=Depth Fade})
  QualitySwitch(Default=@n6, Low=@n5, High=@n6, Medium=@n6, Epic=@n6)
 ]
}
