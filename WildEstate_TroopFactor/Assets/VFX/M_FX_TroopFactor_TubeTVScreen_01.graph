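# Unlit translucent material for the TroopFactor tube TV screen mesh particle in Niagara, showing animated 4x4 flipbook frames over a tintable background.
# Dynamic parameters drive the frame and background color; texture channels plus a procedural band in selected rows mask the composite, which particle color tints before emissive adjustment.
# Opacity combines a ScreenMask-sized sphere mask, particle alpha and a Fresnel airglow, whose region and soft edges come from secondary UV channels.
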Material /WildEstate_TroopFactor/Assets/VFX/M_FX_TroopFactor_TubeTVScreen_01
graph {
 properties {
  bAutomaticallySetUsageInEditor=false
  bUsedWithNiagaraMeshParticles=true
  blendMode=BLEND_Translucent
  previewMesh=&/SM_FX_TroopFactor_TubeTVScreen_01
  shadingModel=MSM_Unlit
  usedShadingModels=Unlit
 }
 nodes [
  n1 = DynamicParameter() {parameterIndex=1, defaultValue=<rgba 1 0.5 0 1>, paramNames [BG Color R, BG Color G, BG Color B, Na]}
  n2 = Constant() {r=4}
  n3 = MaterialFunctionCall(
   "Animation  Phase (0-1)"=DynamicParameter() {paramNames [ScreenMask, SubUV_Animation, New Sub UV Anim, Param4]}.New Sub UV Anim
   Number of Rows=@n2
   Number of Columns=@n2
  ) {materialFunction=&FlipBook}
  n4 = Power(Base=Constant() {r=4}, Exp=ScalarParameter() {defaultValue=-4, parameterName=MipBias})
  n5 = TextureSample(
   UVs=@n3.UVs
   Tex=RecordTextureStreamingInfo!(Tex=TextureObject() {texture=&/T_TroopFactorFlip_new3_copy}, UVs=Multiply(A=@n3.UVs, B=@n4)).Tex
  ) {texture=&T_TroopFactorFlip_new3_copy}
  n6 = Add(A=@n5.R, B=@n5.G)
  n7 = Floor(None=MaterialFunctionCall(UV Channel=Multiply(A=@n3.UVs) {constB=4}) {materialFunction=&LinearGradient}.VGradient)
  n8 = Saturate(None=Subtract(A=Multiply(A=@n7)) {constB=2})
  n9 = MaterialFunctionCall!(
   Width=Constant() {r=0.125}
   Sharpness=Constant() {r=1}
   Offset=Constant() {r=-0.125}
   Direction Switch=StaticBool() {value=true}
   Compare
   Input Coordinates=@n3.UVs
  ) {materialFunction=/Engine/Functions/Engine_MaterialFunctions03/Procedurals/GeneratedBand}
  n10 = SmoothStep(
   Min=ScalarParameter() {parameterName=Smoothstep Mask Low}
   Max=ScalarParameter() {defaultValue=0.01, parameterName=Smoothstep Mask High}
   Value=Add(A=Add(A=@n6, B=@n5.B), B=Multiply(A=@n8, B=@n9.Result))
  ) {constMax=0.05, constMin=0.0001}
  n11 = LinearInterpolate(A=@n1.RGB, B=@n5.RGB, Alpha=Saturate(None=Multiply(A=@n10) {constB=2}))
  n12 = Constant() {r=1}
  n13 = ParticleColor()
  MaterialFunctionCall(Emissive=Multiply(A=StaticSwitchParameter(True=@n11, False=@n12) {group=Flipbook, parameterName=UseTexture}, B=@n13.RGB)) {materialFunction=&MF_FX_EmissiveAdjustments}
  n14 = Constant() {r=0.5}
  n15 = DynamicParameter() {paramNames [ScreenMask, SubUV_Animation, New Sub UV Anim, Param4]}
  n16 = ScalarParameter() {defaultValue=0.9, parameterName=SphereMask_Hardness}
  n17 = TextureCoordinate()
  n18 = Constant() {r=1}
  n19 = TextureCoordinate() {coordinateIndex=1}
  n20 = ComponentMask(None=@n19) {g=true}
  n21 = OneMinus(None=@n20)
  n22 = Power(Base=Clamp(None=Multiply(A=@n21) {constB=2}), Exp=ScalarParameter() {defaultValue=3, parameterName=Airglow_EXP})
  n23 = ScalarParameter() {defaultValue=0.2, parameterName=Airglow_Intensity}
  n24 = ScalarParameter() {defaultValue=1, parameterName=Fresnel_Brightness}
  n25 = ScalarParameter() {defaultValue=2, parameterName=Fresnel_Exp}
  n26 = Multiply(A=ComponentMask(None=@n19) {r=true}) {constB=0.5}
  n27 = ScalarParameter() {defaultValue=2, parameterName=Airglow_Edge_Hardness}
  n28 = Saturate(None=Multiply(A=Sine(None=@n26), B=@n27))
  n29 = ScalarParameter() {defaultValue=2, parameterName=Airglow_Edge_Exp}
  n30 = Multiply(A=ComponentMask(None=TextureCoordinate() {coordinateIndex=2}) {r=true}) {constB=0.5}
  n31 = Saturate(None=Multiply(A=Sine(None=@n30), B=@n27))
  n32 = Multiply(
   A=Multiply(
    A=Multiply(A=@n22, B=@n23)
    B=MaterialFunctionCall(Brightness=@n24, Exponent=@n25) {materialFunction=&MF_Fresnel_FX}.Fresnel
   )
   B=Multiply(A=Power(Base=@n28, Exp=@n29), B=Power(Base=@n31, Exp=@n29))
  )
  n33 = Floor(None=Multiply(A=@n20) {constB=2})
  Multiply(A=Multiply(A=SphereMask(A=@n17, B=@n14, Radius=@n15.ScreenMask, Hardness=@n16), B=LinearInterpolate(A=@n18, B=@n32, Alpha=@n33)), B=@n13.A)
  n34 = Multiply(A=TextureCoordinate(), B=ScalarParameter() {defaultValue=0.01, parameterName=DDXY Mip}) {constB=0.5}
  TextureSample!(UVs=@n3.UVs, Tex, "DDX(UVs)"=DDX(Value=@n34), "DDY(UVs)"=DDY(Value=@n34), Apply View MipBias) {mipValueMode=TMVM_Derivative, texture=&T_TroopFactorFlip_new3_copy}
 ]
}
