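# Idle Niagara effect for the TroopFactor TV, with pulsing antenna and box glows plus an animated screen.
# Two antenna spawn groups place mirrored, counter-rotating sprites, and both glow emitters scale alpha by time of day.
# The screen is one persistent mesh particle with looping lifetime.
# Its age drives an int-array lookup as a hand-sequenced frame index over 16, plus stepped color changes, both fed to dynamic material parameters.
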NiagaraSystem /WildEstate_TroopFactor/Assets/VFX/NS_TroopFactor_TV_Idle
dependencies {
 usedDataInterfaces [&NiagaraDataInterfaceCurve, &NiagaraDataInterfaceArrayInt32, &NiagaraDataInterfaceColorCurve]
 usedDynamicInputs []
 usedModules [&SystemState]
}
emitters {
 Glow_Antenna {
  emitterUpdateScript [
   &EmitterState(Life Cycle Mode=::NewEnumerator1, Loop Behavior=::NewEnumerator1, Loop Duration Mode=::NewEnumerator1, UseLoopDelay=false)
   &SpawnRate(SpawnRate=2, Use Spawn Probability=false)
   &SpawnRate(SpawnGroup=1, SpawnRate=2, Use Spawn Probability=false) {module=SpawnRate001}
  ]
  particleSpawnScript [
   &InitializeParticle(
    Color=<rgba 0.6667 1 0 0.7>
    Lifetime=0.65
    Position Offset=&MaskVectorBySpawnGroup(Masked Vector=<29 0 125>, Passthrough Vector=<-29 1 125>)
    Sprite Rotation Angle=&RandomRangeFloat(Minimum=-1, Maximum=1)
    Sprite Rotation Mode=::NewEnumerator1
    Sprite Size Mode=::NewEnumerator3
    Uniform Sprite Size=50
    UsePositionOffset=true
    Write Material Random=true
   )
   &CameraOffset(Camera Offset Amount=-5)
   /SetVariables(Particles.CameraOffset_Glow=10) {isSetParameters=true}
  ]
  particleUpdateScript [
   &ParticleState()
   &ScaleColor(
    Scale Alpha=&Multiply_Float(
     A=&FloatFromCurve(FloatCurve=DataInterface() {propertyValues {Curve {
         defaultValue=1
         keys [
          {interpMode=RCIM_Cubic, tangentMode=RCTM_User, time=0, value=0}
          {interpMode=RCIM_Cubic, tangentMode=RCTM_User, time=0.5, value=1}
          {tangentMode=RCTM_User, time=1, value=0}
         ]
        }}})
     B=&FNTimeOfDayScalar(Night Scale=0.4, Morning Scale=0.7, Day Scale=0.9, Evening Scale=0.8)
    )
    ScaleA=true
    ScaleRGB=false
   )
   &ScaleSpriteSize(
    Scale Sprite Size Mode=::NewEnumerator0
    Uniform Scale Factor=&FloatFromCurve(FloatCurve=DataInterface() {propertyValues {Curve {
        defaultValue=0.4
        keys [
         {interpMode=RCIM_Cubic, tangentMode=RCTM_User, time=0, value=0.6}
         {interpMode=RCIM_Cubic, tangentMode=RCTM_User, time=0.5, value=1}
         {tangentMode=RCTM_User, time=1, value=0.6}
        ]
       }}})
   )
   &SpriteRotationRate(Rotation Rate=&MaskFloatBySpawnGroup(Masked Float=28.5, Passthrough Float=-25, Spawn Group Mask=$Particles.PartitionParticles.Partition))
   /SetVariables(Particles.ColorStars:LinearColor=&MultiplyLinearColor(
     Linear Color=<rgba 0.3667 1 0.05 0.7>
     Scale Factor=&MakeFloat4FromVectorAndFloat(
      "Vector (RGB)"=<1 1 1>
      "Float (Alpha)"=&Multiply_Float(A=1, B=&FNTimeOfDayScalar(Night Scale=0.75, Morning Scale=0.9, Evening Scale=0.9))
     )
    )) {isSetParameters=true, module=SetVariables_#1}
   /SetVariables(Particles.StarsSize:Vector2f=&Vector2DFromFloat(Value=&FloatFromCurve(
      FloatCurve=DataInterface() {propertyValues {Curve {
         defaultValue=0.4
         keys [
          {interpMode=RCIM_Cubic, tangentMode=RCTM_User, time=0, value=0.6}
          {interpMode=RCIM_Cubic, tangentMode=RCTM_User, time=0.5, value=1}
          {tangentMode=RCTM_User, time=1, value=0.6}
         ]
        }}}
      Scale Curve=35
     ))) {isSetParameters=true, module=SetVariables_#2}
  ]
  rendererClasses [&NiagaraSpriteRendererProperties]
  renderers [
   {rendererClass=&NiagaraSpriteRendererProperties}
   {rendererClass=&NiagaraSpriteRendererProperties, rendererIndex=1}
  ]
 }
 Glow_Box {
  emitterUpdateScript [
   &EmitterState(Life Cycle Mode=::NewEnumerator1, Loop Behavior=::NewEnumerator1, Loop Duration Mode=::NewEnumerator1, UseLoopDelay=false)
   &SpawnRate(SpawnRate=4, Use Spawn Probability=false)
  ]
  particleSpawnScript [
   &InitializeParticle(
    Color=<rgba 5.6 8 0.8 1>
    Lifetime=0.4
    Position Offset=<0 0 40>
    Sprite Rotation Angle=&RandomRangeFloat(Minimum=-1, Maximum=1)
    Sprite Rotation Mode=::NewEnumerator1
    Sprite Size Mode=::NewEnumerator3
    Uniform Sprite Size=180
    UsePositionOffset=true
    Write Material Random=true
   )
   &CameraOffset(Camera Offset Amount=10)
  ]
  particleUpdateScript [
   &ParticleState()
   &ScaleColor(
    Scale Alpha=&Multiply_Float(
     A=&FloatFromCurve(FloatCurve=DataInterface() {propertyValues {Curve {
         defaultValue=1
         keys [
          {arriveTangent=0.0399, interpMode=RCIM_Cubic, leaveTangent=0.0399, tangentMode=RCTM_User, time=0, value=0}
          {interpMode=RCIM_Cubic, tangentMode=RCTM_User, time=0.5, value=1}
          {arriveTangent=-0.0337, tangentMode=RCTM_User, time=1, value=0}
         ]
        }}})
     B=&FNTimeOfDayScalar(Night Scale=0.5, Morning Scale=0.8, Evening Scale=0.8)
    )
    ScaleA=true
    ScaleRGB=false
   )
   &ScaleSpriteSize(
    Scale Sprite Size Mode=::NewEnumerator0
    Uniform Scale Factor=&FloatFromCurve(FloatCurve=DataInterface() {propertyValues {Curve {
        defaultValue=0.4
        keys [
         {interpMode=RCIM_Cubic, tangentMode=RCTM_User, time=0, value=0.95}
         {interpMode=RCIM_Cubic, tangentMode=RCTM_User, time=0.5, value=1}
         {tangentMode=RCTM_User, time=1, value=0.95}
        ]
       }}})
   )
   /SetVariables(Particles.ColorStars:LinearColor=&MultiplyLinearColor(
     Linear Color=<rgba 0.15 1 0.3625 0.7>
     Scale Factor=&MakeFloat4FromVectorAndFloat(
      "Vector (RGB)"=<1 1 1>
      "Float (Alpha)"=&Multiply_Float(A=1, B=&FNTimeOfDayScalar(Night Scale=0.75, Morning Scale=0.9, Evening Scale=0.9))
     )
    )) {isSetParameters=true, module=SetVariables_#1}
   /SetVariables(Particles.StarsSize:Vector2f=&Vector2DFromFloat(Value=&FloatFromCurve(
      FloatCurve=DataInterface() {propertyValues {Curve {
         defaultValue=0.4
         keys [
          {interpMode=RCIM_Cubic, tangentMode=RCTM_User, time=0, value=0.6}
          {interpMode=RCIM_Cubic, tangentMode=RCTM_User, time=0.5, value=1}
          {tangentMode=RCTM_User, time=1, value=0.6}
         ]
        }}}
      Scale Curve=35
     ))) {isSetParameters=true, module=SetVariables_#2}
  ]
  rendererClasses [&NiagaraSpriteRendererProperties]
  renderers [{rendererClass=&NiagaraSpriteRendererProperties}]
 }
 Screen {
  emitterUpdateScript [
   &EmitterState()
   &SpawnBurst_Instantaneous(Use Spawn Probability=false, UseLoopCountLimit=true)
  ]
  particleSpawnScript [
   &InitializeParticle(Color=<rgba 0.8 0.8 0.8 1>, Lifetime=3, Mesh Scale Mode=::NewEnumerator3, UsePositionOffset=true, Write Material Random=true)
   &InitialMeshOrientation(Mesh Orientation Mode=::NewEnumerator1, Use Rotation Vector=false)
   &CameraOffset(Camera Offset Amount=1)
  ]
  particleUpdateScript [
   &ParticleState(Kill Particles When Lifetime Has Elapsed=false, Loop Particles Lifetime=true)
   &DynamicMaterialParameters(
    Index 0 Param 1=1
    Index 0 Param 2=&FloatFromCurve(FloatCurve=DataInterface() {propertyValues {Curve {defaultValue=1, keys [{tangentMode=RCTM_User, time=0, value=0}, {tangentMode=RCTM_User, time=0.55, value=0.99}]}}})
    Index 0 Param 3=&Divide_Float(
     A=&MakeFloatFromInt(Int=&SelectIntFromArray(
       Direct Array Index=&MakeIntFromFloat(Float=&FloatFromCurve(FloatCurve=DataInterface() {propertyValues {Curve {keys [<at 0 0>,<at 0.55 0.99>]}}}, Scale Curve=36))
       Int Selection Array=DataInterface() {propertyValues {IntData [0,0,0,1,2,3,4,5,2,3,4,5,2,3,4,5,6,7,7,7,8,9,10,11,8,9,10,11,8,9,10,11,11,12,13,14]}}
      ))
     B=16
    )
    Index 1 Linear Color:LinearColor=&ColorFromCurve(
     ColorCurve=DataInterface() {propertyValues {
       bUseLUT=false
       AlphaCurve {defaultValue=1, keys []}
       BlueCurve {keys [<at 0 0.0409 RCIM_Constant>,<at 0.1111 0 RCIM_Constant>,<at 0.2188 0.0452 RCIM_Constant>,<at 0.3125 0 RCIM_Constant>,<at 0.5625 0.5972 RCIM_Constant>,<at 0.6562 0.4621 RCIM_Constant>,<at 0.7812 0 RCIM_Constant>,<at 0.972 0.0908 RCIM_Constant>]}
       GreenCurve {keys [<at 0 0.4072 RCIM_Constant>,<at 0.1111 0.7605 RCIM_Constant>,<at 0.2188 0.1746 RCIM_Constant>,<at 0.3125 0 RCIM_Constant>,<at 0.5625 0.1046 RCIM_Constant>,<at 0.6562 0.1095 RCIM_Constant>,<at 0.7812 0.7605 RCIM_Constant>,<at 0.972 0.6867 RCIM_Constant>]}
       RedCurve {keys [<at 0 0.0222 RCIM_Constant>,<at 0.1111 0.9047 RCIM_Constant>,<at 0.2188 1 RCIM_Constant>,<at 0.3125 0.5906 RCIM_Constant>,<at 0.5625 0.0908 RCIM_Constant>,<at 0.6562 0.4969 RCIM_Constant>,<at 0.7812 0.9047 RCIM_Constant>,<at 0.972 0.4793 RCIM_Constant>]}
      }}
     CurveIndex=&FloatFromCurve(FloatCurve=DataInterface() {propertyValues {Curve {keys [<at 0 0>,<at 0.55 0.9999>]}}})
    )
    Module.Index0Float0Write=true
    Module.Index0Float1Write=true
    Module.Index0Float2Write=true
    Module.Index0Float3Write=true
    Write Parameter Index 1=::NewEnumerator5
   )
  ]
  rendererClasses [&NiagaraMeshRendererProperties]
  renderers [{rendererClass=&NiagaraMeshRendererProperties, rendererIndex=0}]
 }
}
summary {
 systemName=NS_TroopFactor_TV_Idle
 emitters [
  {emitterName=Glow_Antenna, rendererClasses [&NiagaraSpriteRendererProperties]}
  {emitterName=Glow_Box, rendererClasses [&NiagaraSpriteRendererProperties]}
  {emitterName=Screen, rendererClasses [&NiagaraMeshRendererProperties]}
 ]
 userVariables [
  User.bIsFrontEnd {type=NiagaraBool}
  User.bIsFrontEndPreview {type=NiagaraBool}
  User.bisHandHidden {type=NiagaraBool}
  User.bKillReaction {type=NiagaraBool}
  User.IsSkydiving {type=NiagaraBool}
 ]
}
