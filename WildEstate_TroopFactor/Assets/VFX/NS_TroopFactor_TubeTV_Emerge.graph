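# Niagara one-shot effect for the TroopFactor TubeTV's emerge moment, apparently flashing its screen with a surrounding glow.
# Burst emitters spawn a Screen mesh and an offset Airglow sprite, sharing a rise-hold-fade curve that drives the screen's material parameter and the glow's size and alpha.
# Airglow alpha scales by time of day and writes green light attributes, though only a sprite renderer is listed.
# A looping, two-sided AntennaGlow emitter is disabled.
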NiagaraSystem /WildEstate_TroopFactor/Assets/VFX/NS_TroopFactor_TubeTV_Emerge
dependencies {usedDataInterfaces [&NiagaraDataInterfaceCurve], usedDynamicInputs [], usedModules [&SystemState]}
emitters {
 Airglow {
  emitterUpdateScript [
   &EmitterState()
   &SpawnBurst_Instantaneous(Use Spawn Probability=false, UseLoopCountLimit=false)
  ]
  particleSpawnScript [
   &InitializeParticle(
    Color=<rgba 1 1 1 0.2>
    Lifetime=1.5
    Position Offset=<-5 30 43>
    Sprite Size Mode=::NewEnumerator3
    Uniform Sprite Size=225
    UsePositionOffset=true
    Write Material Random=true
   )
   &CameraOffset(Camera Offset Amount=50)
  ]
  particleUpdateScript [
   &ParticleState()
   &ScaleColor(
    Scale Alpha=&Multiply_Float(
     A=&Multiply_Float(
      A=&FloatFromCurve(FloatCurve=DataInterface() {propertyValues {Curve {
          defaultValue=1
          keys [
           {arriveTangent=12.3233, interpMode=RCIM_Cubic, leaveTangent=12.3233, tangentMode=RCTM_User, time=0, value=0}
           {arriveTangent=-0.0033, interpMode=RCIM_Cubic, leaveTangent=-0.0033, tangentMode=RCTM_User, time=0.1408, value=0.9847}
           {arriveTangent=0.0933, interpMode=RCIM_Cubic, leaveTangent=-10.2116, tangentMode=RCTM_Break, time=0.695, value=1}
           {arriveTangent=-3.279, interpMode=RCIM_Cubic, leaveTangent=-3.279, time=0.7953, value=0.2004}
           {arriveTangent=-0.5634, interpMode=RCIM_Cubic, leaveTangent=-0.5634, tangentMode=RCTM_User, time=1, value=0}
          ]
         }}})
      B=&FNTimeOfDayScalar(Night Scale=0.75, Morning Scale=0.9, Evening Scale=0.9)
     )
     B=1
    )
    ScaleA=true
    ScaleRGB=true
   )
   &ScaleSpriteSize(
    Scale Sprite Size Mode=::NewEnumerator0
    Uniform Scale Factor=&FloatFromCurve(FloatCurve=DataInterface() {propertyValues {Curve {
        defaultValue=1
        keys [
         {arriveTangent=12.3233, interpMode=RCIM_Cubic, leaveTangent=12.3233, tangentMode=RCTM_User, time=0, value=0}
         {arriveTangent=-0.0033, interpMode=RCIM_Cubic, leaveTangent=-0.0033, tangentMode=RCTM_User, time=0.1408, value=0.9847}
         {arriveTangent=0.0933, interpMode=RCIM_Cubic, leaveTangent=-10.2116, tangentMode=RCTM_Break, time=0.695, value=1}
         {arriveTangent=-3.279, interpMode=RCIM_Cubic, leaveTangent=-3.279, time=0.7953, value=0.2004}
         {arriveTangent=-0.5634, interpMode=RCIM_Cubic, leaveTangent=-0.5634, tangentMode=RCTM_User, time=1, value=0}
        ]
       }}})
   )
   &FN_Light_Attributes(Color=<rgba 0 10 0 1>, Create Color Attribute=true, Create Exponent Attribute=true, Create Radius Attribute=true, Exponent=1, Radius=60)
   &SolveForcesAndVelocity(Clamp Velocity=false, Limit Acceleration=false)
  ]
  rendererClasses [&NiagaraSpriteRendererProperties]
  renderers [{rendererClass=&NiagaraSpriteRendererProperties}]
 }
 AntennaGlow {
  bEnabled=false
  emitterUpdateScript [
   &EmitterState(Life Cycle Mode=::NewEnumerator1, Loop Behavior=::NewEnumerator1, Loop Duration Mode=::NewEnumerator1, UseLoopDelay=false)
   &SpawnRate(SpawnRate=2, Use Spawn Probability=false)
   &SpawnRate(SpawnGroup=1, SpawnRate=2, Use Spawn Probability=false) {module=SpawnRate001}
  ]
  particleSpawnScript [
   &InitializeParticle(
    Color=<rgba 0 1 0 0.7>
    Lifetime=0.65
    Position Offset=&MaskVectorBySpawnGroup(Masked Vector=<29 0 125>, Passthrough Vector=<-29 1 125>)
    Sprite Rotation Angle=&RandomRangeFloat(Minimum=-1, Maximum=1)
    Sprite Rotation Mode=::NewEnumerator1
    Sprite Size Mode=::NewEnumerator3
    Uniform Sprite Size=50
    UsePositionOffset=true
    Write Material Random=true
   )
   &CameraOffset(Camera Offset Amount=-5)
   /SetVariables(Particles.CameraOffset_Glow=10) {isSetParameters=true}
  ]
  particleUpdateScript [
   &ParticleState()
   &ScaleColor(
    Scale Alpha=&Multiply_Float(
     A=&FloatFromCurve(FloatCurve=DataInterface() {propertyValues {Curve {
         defaultValue=1
         keys [
          {interpMode=RCIM_Cubic, tangentMode=RCTM_User, time=0, value=0}
          {interpMode=RCIM_Cubic, tangentMode=RCTM_User, time=0.5, value=1}
          {tangentMode=RCTM_User, time=1, value=0}
         ]
        }}})
     B=&FNTimeOfDayScalar(Night Scale=0.4, Morning Scale=0.7, Day Scale=0.9, Evening Scale=0.8)
    )
    ScaleA=true
    ScaleRGB=false
   )
   &ScaleSpriteSize(
    Scale Sprite Size Mode=::NewEnumerator0
    Uniform Scale Factor=&FloatFromCurve(FloatCurve=DataInterface() {propertyValues {Curve {
        defaultValue=0.4
        keys [
         {interpMode=RCIM_Cubic, tangentMode=RCTM_User, time=0, value=0.6}
         {interpMode=RCIM_Cubic, tangentMode=RCTM_User, time=0.5, value=1}
         {tangentMode=RCTM_User, time=1, value=0.6}
        ]
       }}})
   )
   &SpriteRotationRate(Rotation Rate=&MaskFloatBySpawnGroup(Masked Float=28.5, Passthrough Float=-25, Spawn Group Mask=$Particles.PartitionParticles.Partition))
   /SetVariables(Particles.ColorStars:LinearColor=&MultiplyLinearColor(
     Linear Color=<rgba 0.15 1 0.3625 0.7>
     Scale Factor=&MakeFloat4FromVectorAndFloat(
      "Vector (RGB)"=<1 1 1>
      "Float (Alpha)"=&Multiply_Float(A=1, B=&FNTimeOfDayScalar(Night Scale=0.75, Morning Scale=0.9, Evening Scale=0.9))
     )
    )) {isSetParameters=true}
   /SetVariables(Particles.StarsSize:Vector2f=&Vector2DFromFloat(Value=&FloatFromCurve(
      FloatCurve=DataInterface() {propertyValues {Curve {
         defaultValue=0.4
         keys [
          {interpMode=RCIM_Cubic, tangentMode=RCTM_User, time=0, value=0.6}
          {interpMode=RCIM_Cubic, tangentMode=RCTM_User, time=0.5, value=1}
          {tangentMode=RCTM_User, time=1, value=0.6}
         ]
        }}}
      Scale Curve=35
     ))) {isSetParameters=true}
  ]
  rendererClasses [&NiagaraSpriteRendererProperties]
  renderers [
   {rendererClass=&NiagaraSpriteRendererProperties}
   {rendererClass=&NiagaraSpriteRendererProperties, rendererIndex=1}
  ]
 }
 Screen {
  emitterUpdateScript [
   &EmitterState()
   &SpawnBurst_Instantaneous(Use Spawn Probability=false, UseLoopCountLimit=false)
  ]
  particleSpawnScript [
   &InitializeParticle(Lifetime=1.5, Mesh Scale Mode=::NewEnumerator3, Position Offset=<0 0 -1>, UsePositionOffset=true, Write Material Random=true)
   &InitialMeshOrientation(Mesh Orientation Mode=::NewEnumerator1, Use Rotation Vector=false)
   &CameraOffset(Camera Offset Amount=5)
  ]
  particleUpdateScript [
   &ParticleState()
   &ScaleSpriteSize(Uniform Curve Sprite Scale=DataInterface() {propertyValues {Curve {keys [{tangentMode=RCTM_User, time=0, value=0}, <at 0.1 0.7>, {tangentMode=RCTM_User, time=1, value=1}]}}})
   &DynamicMaterialParameters(
    Index 0 Param 1=&FloatFromCurve(FloatCurve=DataInterface() {propertyValues {Curve {
        defaultValue=1
        keys [
         {arriveTangent=12.3233, interpMode=RCIM_Cubic, leaveTangent=12.3233, tangentMode=RCTM_User, time=0, value=0}
         {arriveTangent=-0.0033, interpMode=RCIM_Cubic, leaveTangent=-0.0033, tangentMode=RCTM_User, time=0.1408, value=0.9847}
         {arriveTangent=0.0933, interpMode=RCIM_Cubic, leaveTangent=-10.2116, tangentMode=RCTM_Break, time=0.695, value=1}
         {arriveTangent=-3.279, interpMode=RCIM_Cubic, leaveTangent=-3.279, time=0.7953, value=0.2004}
         {arriveTangent=-0.5634, interpMode=RCIM_Cubic, leaveTangent=-0.5634, tangentMode=RCTM_User, time=1, value=0}
        ]
       }}})
    Module.Index0Float0Write=true
    Module.Index0Float1Write=true
    Module.Index0Float2Write=true
    Module.Index0Float3Write=true
   )
  ]
  rendererClasses [&NiagaraMeshRendererProperties]
  renderers [{rendererClass=&NiagaraMeshRendererProperties, rendererIndex=0}]
 }
}
summary {
 systemName=NS_TroopFactor_TubeTV_Emerge
 emitters [
  {emitterName=Screen, rendererClasses [&NiagaraMeshRendererProperties]}
  {emitterName=Airglow, rendererClasses [&NiagaraSpriteRendererProperties]}
  {bEnabled=false, emitterName=AntennaGlow, rendererClasses [&NiagaraSpriteRendererProperties]}
 ]
 userVariables []
}
